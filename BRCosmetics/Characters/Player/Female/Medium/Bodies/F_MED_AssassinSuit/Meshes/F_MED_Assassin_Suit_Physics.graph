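# Physics asset for the female medium Assassin Suit body, set up for secondary motion on its front flap, back jacket flaps, necklace and pendant.
# A three-segment front flap chain hangs from a base on spine_02, two back flaps from bases on pelvis, and the necklace and pendant from spine_05.
# These dangling joints have angular limits, while body and base joints rotate freely.
# Pelvis, spine and right thigh supply body shapes, with no left thigh body.
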PhysicsAsset /BRCosmetics/Characters/Player/Female/Medium/Bodies/F_MED_AssassinSuit/Meshes/F_MED_Assassin_Suit_Physics
bodies {
 dyn_frontFlapBase [{shapeType=Box, extentX=1.01, extentY=1.01, extentZ=1.01, rotation=<rot 0 -90 0>}]
 dyn_frontFlap_01 [{shapeType=Box, extentX=2.0248, extentY=6.5166, extentZ=15.0987, center=<3.327 0.2158 -0.093>, rotation=<rot -2.7337 -89.9999 -6.3864>}]
 dyn_frontFlap_02 [{shapeType=Box, extentX=1.3864, extentY=6.2279, extentZ=11.8735, center=<2.4102 0 0>, rotation=<rot 0 -90 0>}]
 dyn_frontFlap_03 [{shapeType=Box, extentX=1.3231, extentY=3.9073, extentZ=8.1245, center=<1.6914 0.0001 0>, rotation=<rot 0 -90 0>}]
 dyn_jacketFlap_backBase_l [{shapeType=Box, extentX=1.01, extentY=1.01, extentZ=1.01, rotation=<rot 0 -90 0>}]
 dyn_jacketFlap_backBase_r [{shapeType=Box, extentX=1.01, extentY=1.01, extentZ=1.01, rotation=<rot 0 -90 0>}]
 dyn_jacketFlap_back_l [{shapeType=Box, extentX=2.6084, extentY=5.8519, extentZ=9.9766, center=<2.25 0 0>, rotation=<rot 0 -90 0>}]
 dyn_jacketFlap_back_r [{shapeType=Box, extentX=2.7217, extentY=6.5843, extentZ=9.9766, center=<2.25 0 0>, rotation=<rot 0 -90 0>}]
 dyn_necklace [{shapeType=Capsule, length=8.2762, radius=2.0935, center=<1.0886 0.0001 0>, rotation=<rot 90 0 0>}]
 dyn_pendant [{shapeType=Capsule, length=1.4895, radius=1.071, center=<1.551 -0.0049 -0.0302>, rotation=<rot 0 90.1957 -88.9911>}]
 pelvis [{shapeType=Capsule, length=11.7537, radius=8.1688, shapeName=pelvis_capsule, center=<0 0.5084 0>}]
 spine_02 [{shapeType=Capsule, length=13.6213, radius=8.8845, center=<0.921 1.6797 0.41>, rotation=<rot -2.1955 -87.2226 -83.6421>}]
 spine_05 [
  {shapeType=Capsule, length=25.3338, radius=4.6662, center=<-0.2574 8.5211 7.4595>, rotation=<rot -1.2207 67.7766 104.524>}
  {shapeType=Capsule, length=25.3338, radius=4.6662, center=<1.8357 8.4094 -6.8701>, rotation=<rot -1.2207 67.7746 75.4359>}
  {shapeType=Box, extentX=9.1573, extentY=5.3182, extentZ=13.6718, center=<0.3178 5.7797 -0.0001>, rotation=<rot 0 -36.038 0>}
 ]
 thigh_r [{
   shapeType=Capsule
   length=22.2274
   radius=8.0124
   shapeName=thigh_r_capsule
   center=<22.7939 0.2921 -0.0136>
   rotation=<rot 90 89.4328 86.509>
  }]
}
constraints [
 {bone1Name=dyn_jacketFlap_back_l, bone2Name=dyn_jacketFlap_backBase_l, swing1LimitDegrees=20, swing2LimitDegrees=5, twistLimitDegrees=2}
 {bone1Name=dyn_jacketFlap_back_r, bone2Name=dyn_jacketFlap_backBase_r, swing1LimitDegrees=20, swing2LimitDegrees=5, twistLimitDegrees=2}
 {bone1Name=dyn_frontFlap_01, bone2Name=dyn_frontFlapBase, swing1LimitDegrees=30, swing2LimitDegrees=5, twistLimitDegrees=5}
 {bone1Name=dyn_frontFlap_02, bone2Name=dyn_frontFlap_01, swing1LimitDegrees=30, swing2LimitDegrees=15, twistLimitDegrees=5}
 {bone1Name=dyn_frontFlap_03, bone2Name=dyn_frontFlap_02, swing1LimitDegrees=30, swing2LimitDegrees=15, twistLimitDegrees=5}
 {
  bone1Name=dyn_frontFlapBase
  bone2Name=spine_02
  swing1LimitDegrees=45
  swing1Motion=Free
  swing2LimitDegrees=45
  swing2Motion=Free
  twistLimitDegrees=45
  twistMotion=Free
 }
 {
  bone1Name=spine_05
  bone2Name=spine_02
  swing1LimitDegrees=45
  swing1Motion=Free
  swing2LimitDegrees=45
  swing2Motion=Free
  twistLimitDegrees=45
  twistMotion=Free
 }
 {bone1Name=dyn_necklace, bone2Name=spine_05, swing1LimitDegrees=10, swing2LimitDegrees=40, twistLimitDegrees=1}
 {bone1Name=dyn_pendant, bone2Name=dyn_necklace, swing1LimitDegrees=40, swing2LimitDegrees=45, twistLimitDegrees=2}
 {
  bone1Name=spine_02
  bone2Name=pelvis
  swing1LimitDegrees=45
  swing1Motion=Free
  swing2LimitDegrees=45
  swing2Motion=Free
  twistLimitDegrees=45
  twistMotion=Free
 }
 {
  bone1Name=thigh_r
  bone2Name=pelvis
  swing1LimitDegrees=45
  swing1Motion=Free
  swing2LimitDegrees=45
  swing2Motion=Free
  twistLimitDegrees=45
  twistMotion=Free
 }
 {
  bone1Name=dyn_jacketFlap_backBase_l
  bone2Name=pelvis
  swing1LimitDegrees=45
  swing1Motion=Free
  swing2LimitDegrees=45
  swing2Motion=Free
  twistLimitDegrees=45
  twistMotion=Free
 }
 {
  bone1Name=dyn_jacketFlap_backBase_r
  bone2Name=pelvis
  swing1LimitDegrees=45
  swing1Motion=Free
  swing2LimitDegrees=45
  swing2Motion=Free
  twistLimitDegrees=45
  twistMotion=Free
 }
]
