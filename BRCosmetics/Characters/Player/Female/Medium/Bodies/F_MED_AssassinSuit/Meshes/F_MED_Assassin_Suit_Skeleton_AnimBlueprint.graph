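# Animation Blueprint for the female medium assassin suit, adding secondary motion to jacket flaps, dagger and zipper.
# It copies the attached parent's pose, then runs pelvis-space RigidBody simulation, additive flap-base offsets, AnimDynamics springs and a spine/pelvis-weighted zipper-base constraint.
# The event graph maps thigh swing angles to flap offset alphas and inversely to flap dynamics alphas.
# The left flap gets an extra push at high swing.
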AnimBlueprint /BRCosmetics/Characters/Player/Female/Medium/Bodies/F_MED_AssassinSuit/Meshes/F_MED_Assassin_Suit_Skeleton_AnimBlueprint
target_skeleton=&/F_MED_Assassin_Suit_Skeleton
graphs [
 AnimGraph {nodes [
   n1 = RigidBody(ComponentPose=Local To Component(LocalPose=Copy Pose From Mesh() {bUseAttachedParent=true})) {
    bClampLinearTranslationLimitToRefPose=true
    bOverrideWorldGravity=true
    overridePhysicsAsset=&/F_MED_Assassin_Suit_Physics
    simulationSpace=BaseBoneSpace
    alphaBoolBlend {blendOption=Linear}
    baseBoneRef {boneName=pelvis}
    componentAppliedLinearAccClamp=<2000 2000 2000>
    componentLinearAccScale=<0.4 0.4 0.5>
    componentLinearVelScale=<0.3 0.3 0.4>
    overrideWorldGravity {z=-980}
    simSpaceSettings {
     maxAngularAcceleration=500
     maxAngularVelocity=3
     maxLinearAcceleration=400
     maxLinearVelocity=250
     velocityScaleZ=0
     worldAlpha=1
     externalLinearDragV=<0.1 0.1>
    }
   }
   n2 = "Transform (Modify) Bone - Bone: dyn_jacketFlap_frontBase_l"(
    ComponentPose=@n1
    Alpha=Get legSwingDelta_l() -> (legSwingDelta_l:double)
    Translation="-9,-2,-7.000"
    Rotation="38.636,-71.952,-5"
    Scale="1,1,1"
   ) {
    rotationMode=BMM_Additive
    rotationSpace=BCS_BoneSpace
    translationMode=BMM_Additive
    translationSpace=BCS_BoneSpace
    alphaBoolBlend {blendOption=Linear}
    boneToModify {boneName=dyn_jacketFlap_frontBase_l}
    rotation=<rot 38.636 -71.952 -5>
    translation=<-9 -2 -7>
   }
   n3 = "Transform (Modify) Bone - Bone: dyn_jacketFlap_frontBase_l"(
    ComponentPose=@n2
    Alpha=Get legSwingDeltaSuper_l() -> (legSwingDeltaSuper_l:double)
    Translation="0.5,-2,0"
    Rotation="0,-10,0"
    Scale="1,1,1"
   ) {
    rotationMode=BMM_Additive
    rotationSpace=BCS_BoneSpace
    translationMode=BMM_Additive
    translationSpace=BCS_BoneSpace
    alphaBoolBlend {blendOption=Linear}
    boneToModify {boneName=dyn_jacketFlap_frontBase_l}
    rotation {yaw=-10}
    translation=<0.5 -2>
   }
   n4 = "Transform (Modify) Bone - Bone: dyn_jacketFlap_frontBase_r"(
    ComponentPose=@n3
    Alpha=Get legSwingDelta_r() -> (legSwingDelta_r:double)
    Translation="-5.567,2.137,2.874"
    Rotation="4.967,-91.497,34.116"
    Scale="1,1,1"
   ) {
    rotationMode=BMM_Additive
    rotationSpace=BCS_BoneSpace
    translationMode=BMM_Additive
    translationSpace=BCS_BoneSpace
    alphaBoolBlend {blendOption=Linear}
    boneToModify {boneName=dyn_jacketFlap_frontBase_r}
    rotation=<rot 4.967 -91.497 34.1163>
    translation=<-5.567 2.137 2.874>
   }
   n5 = Anim Dynamics - Bone: dyn_jacketFlap_front_l(ComponentPose=@n4, Alpha=Get legDynAlpha_l() -> (legDynAlpha_l:double), ExternalForce="0,0,0") {
    angularDampingOverride=1
    angularSpringConstant=500
    bAngularSpring=true
    bGravityOverrideInSimSpace=true
    bOverrideAngularDamping=true
    simulationSpace=RootRelative
    alphaBoolBlend {blendOption=Linear}
    alphaScaleBiasClamp {bClampResult=true, clampMin=0.15}
    boundBone {boneName=dyn_jacketFlap_front_l}
    componentAppliedLinearAccClamp=<1000 1000 1000>
    componentLinearAccScale=<0.3 0.3 0.3>
    componentLinearVelScale=<0.1 0.1 0.1>
    physicsBodyDefinitions [{
      sphereCollisionRadius=0
      boundBone {boneName=dyn_jacketFlap_front_l}
      boxExtents=<10 2 10>
      constraintSetup {angularLimitsMax=<1 5 0>, angularLimitsMin=<-1 -10 -30>, angularTarget=<1 0 0>}
      localJointOffset=<4 0 1.5>
     }]
    relativeSpaceBone {boneName=dyn_jacketFlap_frontBase_l}
   }
   n6 = Anim Dynamics - Bone: dyn_jacketFlap_front_r(ComponentPose=@n5, Alpha=Get legDynAlpha_r() -> (legDynAlpha_r:double), ExternalForce="0,0,0") {
    angularDampingOverride=1
    angularSpringConstant=500
    bAngularSpring=true
    bGravityOverrideInSimSpace=true
    bOverrideAngularDamping=true
    simulationSpace=RootRelative
    alphaBoolBlend {blendOption=Linear}
    alphaScaleBiasClamp {bClampResult=true, clampMax=0.85, clampMin=0.15}
    boundBone {boneName=dyn_jacketFlap_front_r}
    componentAppliedLinearAccClamp=<1000 1000 1000>
    componentLinearAccScale=<0.3 0.3 0.3>
    componentLinearVelScale=<0.1 0.1 0.1>
    physicsBodyDefinitions [{
      sphereCollisionRadius=0
      boundBone {boneName=dyn_jacketFlap_front_r}
      boxExtents=<10 2 10>
      constraintSetup {angularLimitsMax=<1 10 5>, angularLimitsMin=<-1 -5 -20>, angularTarget=<1 0 0>}
      localJointOffset=<4 0 -1.5>
     }]
    relativeSpaceBone {boneName=dyn_jacketFlap_frontBase_r}
   }
   n7 = Anim Dynamics - Bone: dyn_dagger(ComponentPose=@n6, Alpha=1, ExternalForce="0,0,0") {
    angularDampingOverride=1
    angularSpringConstant=500
    bAngularSpring=true
    bGravityOverrideInSimSpace=true
    bOverrideAngularDamping=true
    simulationSpace=RootRelative
    alphaBoolBlend {blendOption=Linear}
    alphaScaleBiasClamp {bClampResult=true}
    boundBone {boneName=dyn_dagger}
    componentAppliedLinearAccClamp=<1000 1000 1000>
    componentLinearAccScale=<0.3 0.3 0.3>
    componentLinearVelScale=<0.1 0.1 0.1>
    physicsBodyDefinitions [{
      sphereCollisionRadius=0
      boundBone {boneName=dyn_dagger}
      boxExtents=<15 4 15>
      constraintSetup {angularLimitsMax=<4 15 3>, angularLimitsMin=<-4 -15 -3>, angularTarget=<1 0 0>}
      localJointOffset=<4 0 0>
     }]
    relativeSpaceBone {boneName=thigh_r}
   }
   n8 = "Constraint - dyn_zipper_Base (PPPPP)"!(
    ComponentPose=@n7
    Alpha=1
    ConstraintWeights_0=0.172
    ConstraintWeights_1:float=0.3632
    ConstraintWeights_2:float=0.1568
    ConstraintWeights_3:float=0.2918
    ConstraintWeights_4:float=0.0162
   ) -> (Pose) {
    alphaBoolBlend {blendOption=Linear}
    boneToModify {boneName=dyn_zipper_Base}
    constraintSetup [
     {targetBone {boneName=spine_01}}
     {targetBone {boneName=spine_02}}
     {targetBone {boneName=spine_03}}
     {targetBone {boneName=spine_04}}
     {targetBone {boneName=pelvis}}
    ]
    constraintWeights [1,1,1,1,1]
   }
   n9 = Anim Dynamics - Bone: dyn_zipper(ComponentPose=@n8, Alpha=1) {
    angularDampingOverride=1
    angularSpringConstant=500
    bAngularSpring=true
    bGravityOverrideInSimSpace=true
    bOverrideAngularDamping=true
    simulationSpace=RootRelative
    alphaBoolBlend {blendOption=Linear}
    alphaScaleBiasClamp {bClampResult=true}
    boundBone {boneName=dyn_zipper}
    componentAppliedLinearAccClamp=<1000 1000 1000>
    componentLinearAccScale=<0.3 0.3 0.3>
    componentLinearVelScale=<0.1 0.1 0.1>
    physicsBodyDefinitions [{
      sphereCollisionRadius=0
      boundBone {boneName=dyn_zipper}
      boxExtents=<8 4 8>
      constraintSetup {angularLimitsMax=<2 2 2>, angularLimitsMin=<-2 -2 -20>, angularTarget=<1 0 0>}
      localJointOffset=<2 0 0>
     }]
    relativeSpaceBone {boneName=thigh_r}
   }
   Output Pose(Result=Component To Local(ComponentPose=@n9)) {layerGroup=None}
  ]}
 EventGraph {nodes [
   Event BlueprintUpdateAnimation()
   > n1 = Sequence()
   n2 = GetOwnerSkeletalMeshComponent()
   n3 = MapRangeClamped(Value=GetTwistAndSwingAngleOfDeltaRotationFromRefPose(self=@n2, BoneName=thigh_r).OutSwingAngle, InRangeA=173, InRangeB=68)
   Set legSwingDelta_r(execute=@n1.then_1, legSwingDelta_r:double=@n3) -> (Output_Get:double)
   n4 = MapRangeClamped(Value=GetTwistAndSwingAngleOfDeltaRotationFromRefPose(self=@n2, BoneName=thigh_l).OutSwingAngle, InRangeA=24, InRangeB=119.5)
   > Set legSwingDelta_l(legSwingDelta_l:double=@n4) -> (Output_Get:double)
   > Set legDynAlpha_r(legDynAlpha_r:double=float - float(B=@n3) {is=title K2Node_CallFunction}) -> (Output_Get:double)
   > Set legDynAlpha_l(legDynAlpha_l:double=float - float(B=@n4)) -> (Output_Get:double)
   > Set legPush_l(legPush_l:double="To Float (Boolean)"(InBool="float >= float"(A=@n4, B=0.84) {is=title K2Node_CallFunction})) -> (Output_Get:double)
   > Set legSwingDeltaSuper_l(legSwingDeltaSuper_l:double=float * float(A=MapRangeClamped(Value=@n4, InRangeA=0.84, InRangeB=1), B=Get legPush_l() -> (legPush_l:double))) -> (Output_Get:double)
   TryGetPawnOwner()
  ]}
]
variables [
 isdbnoAlpha {type=number}
 legDynAlpha_l {type=number}
 legDynAlpha_r {type=number}
 legPush_l {type=number}
 legSwingDelta_l {type=number}
 legSwingDelta_r {type=number}
 legSwingDeltaSuper_l {type=number}
]
