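# Physics asset for the female medium Assassin Suit face accessory mesh, covering the secondary motion of its ponytail and loose hair locks.
# Capsule bodies form a four-link ponytail chain from the head, with small locks branching off its middle links and four front locks attached directly to the head.
# Hair joints have tight angular limits, with less twist on the front locks.
# Larger head, neck and upper-spine capsules are linked by free angular constraints.
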PhysicsAsset /BRCosmetics/Characters/Player/Female/Medium/Bodies/F_MED_AssassinSuit/Meshes/Parts/F_MED_Assassin_Suit_FaceAcc_Physics
bodies {
 dyn_frontLockA_l [{shapeType=Capsule, length=2.6948, radius=1.0815, center=<1.4695 0 0>, rotation=<rot 90 0 0>}]
 dyn_frontLockA_r [{shapeType=Capsule, length=2.9293, radius=0.9756, center=<1.6514 0 0>, rotation=<rot 90 0 0>}]
 dyn_frontLockB_l [{shapeType=Capsule, length=3.5009, radius=1.2153, center=<1.9262 0 0>, rotation=<rot 90 0 0>}]
 dyn_frontLockB_r [{shapeType=Capsule, length=1.318, radius=0.505, center=<0.6928 0 0>, rotation=<rot 90 0 0>}]
 dyn_lock1 [{shapeType=Capsule, length=2.7079, radius=0.5866, center=<1.5661 0 0.0001>, rotation=<rot 90 0 0>}]
 dyn_lock2 [{shapeType=Capsule, length=2.311, radius=0.5017, center=<1.0246 0 0>, rotation=<rot 90 0 0>}]
 dyn_lock3 [{shapeType=Capsule, length=2.7229, radius=0.5, center=<1.4774 0 0>, rotation=<rot 90 0 0>}]
 dyn_lock4 [{shapeType=Capsule, length=1.4268, radius=0.505, center=<0.7496 0.0001 0>, rotation=<rot 90 0 0>}]
 dyn_ponytail_01 [{shapeType=Capsule, length=5.6095, radius=4.0098, center=<3.1494 0 0.0001>, rotation=<rot 90 0 0>}]
 dyn_ponytail_02 [{shapeType=Capsule, length=5.3451, radius=3.4648, center=<2.9357 0 0>, rotation=<rot 90 0 0>}]
 dyn_ponytail_03 [{shapeType=Capsule, length=4.149, radius=2.4546, center=<1.9179 0 0>, rotation=<rot 90 0 0>}]
 dyn_ponytail_04 [{shapeType=Capsule, length=2.2588, radius=1.5386, center=<1.7121 0 0>, rotation=<rot 90 0 0>}]
 head [{shapeType=Capsule, length=5.5509, radius=6.6972, center=<4.0404 0.6601 0.4406>, rotation=<rot 5.9772 -73.8407 -89.5139>}]
 neck_01 [{shapeType=Capsule, length=2.6108, radius=5.972, rotation=<rot 90 0 0>}]
 neck_02 [{shapeType=Capsule, length=3.6265, radius=4.9564, rotation=<rot 90 0 0>}]
 spine_05 [{shapeType=Capsule, length=17.2406, radius=11.0537}]
}
constraints [
 {bone1Name=dyn_ponytail_01, bone2Name=head, swing1LimitDegrees=20, swing2LimitDegrees=20, twistLimitDegrees=20}
 {bone1Name=dyn_ponytail_02, bone2Name=dyn_ponytail_01, swing1LimitDegrees=20, swing2LimitDegrees=20, twistLimitDegrees=20}
 {bone1Name=dyn_ponytail_03, bone2Name=dyn_ponytail_02, swing1LimitDegrees=20, swing2LimitDegrees=20, twistLimitDegrees=20}
 {bone1Name=dyn_ponytail_04, bone2Name=dyn_ponytail_03, swing1LimitDegrees=20, swing2LimitDegrees=20, twistLimitDegrees=20}
 {bone1Name=dyn_lock2, bone2Name=dyn_ponytail_03, swing1LimitDegrees=20, swing2LimitDegrees=20, twistLimitDegrees=20}
 {bone1Name=dyn_lock3, bone2Name=dyn_ponytail_03, swing1LimitDegrees=20, swing2LimitDegrees=20, twistLimitDegrees=20}
 {bone1Name=dyn_lock4, bone2Name=dyn_ponytail_03, swing1LimitDegrees=20, swing2LimitDegrees=20, twistLimitDegrees=20}
 {bone1Name=dyn_lock1, bone2Name=dyn_ponytail_02, swing1LimitDegrees=20, swing2LimitDegrees=20, twistLimitDegrees=20}
 {bone1Name=dyn_frontLockA_r, bone2Name=head, swing1LimitDegrees=20, swing2LimitDegrees=20, twistLimitDegrees=5}
 {bone1Name=dyn_frontLockB_r, bone2Name=head, swing1LimitDegrees=20, swing2LimitDegrees=20, twistLimitDegrees=5}
 {bone1Name=dyn_frontLockA_l, bone2Name=head, swing1LimitDegrees=20, swing2LimitDegrees=20, twistLimitDegrees=5}
 {bone1Name=dyn_frontLockB_l, bone2Name=head, swing1LimitDegrees=20, swing2LimitDegrees=20, twistLimitDegrees=5}
 {
  bone1Name=head
  bone2Name=neck_02
  swing1LimitDegrees=45
  swing1Motion=Free
  swing2LimitDegrees=45
  swing2Motion=Free
  twistLimitDegrees=45
  twistMotion=Free
 }
 {
  bone1Name=neck_02
  bone2Name=neck_01
  swing1LimitDegrees=45
  swing1Motion=Free
  swing2LimitDegrees=45
  swing2Motion=Free
  twistLimitDegrees=45
  twistMotion=Free
 }
 {
  bone1Name=neck_01
  bone2Name=spine_05
  swing1LimitDegrees=45
  swing1Motion=Free
  swing2LimitDegrees=45
  swing2Motion=Free
  twistLimitDegrees=45
  twistMotion=Free
 }
]
