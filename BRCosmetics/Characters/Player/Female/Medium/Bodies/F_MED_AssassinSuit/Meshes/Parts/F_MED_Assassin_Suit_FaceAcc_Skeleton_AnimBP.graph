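# Post-process-style Anim Blueprint that adds physics motion to the female Assassin suit's face accessory, including wind lift while skydiving.
# The AnimGraph copies the attached parent mesh's pose, then simulates the accessory with a RigidBody node using its own physics asset.
# Each update turns downward pawn velocity into a randomized, Perlin-remapped vertical external force while skydiving, scaled to a tenth once the parachute opens.
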AnimBlueprint /BRCosmetics/Characters/Player/Female/Medium/Bodies/F_MED_AssassinSuit/Meshes/Parts/F_MED_Assassin_Suit_FaceAcc_Skeleton_AnimBP
target_skeleton=&/F_MED_Assassin_Suit_FaceAcc_Skeleton
graphs [
 AnimGraph {nodes [
   n1 = RigidBody(
    ComponentPose=Local To Component(LocalPose=Copy Pose From Mesh() {bUseAttachedParent=true})
    ExternalForce=Get skydivingForce() -> (skydivingForce:Vector)
   ) {
    bClampLinearTranslationLimitToRefPose=true
    bOverrideWorldGravity=true
    overridePhysicsAsset=&/F_MED_Assassin_Suit_FaceAcc_Physics
    alphaBoolBlend {blendOption=Linear}
    baseBoneRef {boneName=pelvis}
    componentAppliedLinearAccClamp=<2000 2000 2000>
    componentLinearAccScale=<0.6 0.6 0.6>
    componentLinearVelScale=<0.4 0.4 0.4>
    overrideWorldGravity {z=-980}
   }
   Output Pose(Result=Component To Local(ComponentPose=@n1)) {layerGroup=None}
  ]}
 EventGraph {nodes [
   Event BlueprintUpdateAnimation()
   n1 = Get PawnVelocityZ()
   n2 = RandomFloatInRangeFromStream!(
    Stream:"Random Stream Structure (by ref)"=Get random_stream() -> (random_stream:Random Stream Structure)
    Min:float=float * float(A=@n1, B=2.5)
    Max:float=float * float(A=@n1, B=5.5)
   ) -> (ReturnValue:float)
   n3 = SelectFloat(
    A=Make Perlin Noise and Remap!(Value:float=@n2, RangeOutMin:float=-100, RangeOutMax:float=30000) -> (ReturnValue:float)
    B=0
    bPickA=AND Boolean(A="float < float"!(A:double=@n2, B:double=0) -> (ReturnValue:bool) {is=title K2Node_CallFunction}, B=Get bIsSkyDiving())
   )
   n4 = float * float(A=@n3, B=0.1)
   n5 = Get bIsParachuteOpened()
   > Set skydivingForce(skydivingForce:Vector=MakeVector(Y=0, Z=SelectFloat(A=@n4, B=@n3, bPickA=@n5))) -> (Output_Get:Vector)
   TryGetPawnOwner()
   Get PawnVelocityZ()
   Set random_stream(random_stream:Random Stream Structure=MakeRandomStream!(InitialSeed:int=0) -> (ReturnValue:Random Stream Structure)) -> (Output_Get:Random Stream Structure)
  ]}
]
variables [
 random_stream {type=RandomStream, default {initialSeed=0, seed=0}}
 skydivingForce {type=Vector}
]
